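# Niagara system for the Glitched vehicle boost cosmetic, drawing ribbon trails, sprite particles and a mesh-and-sprite thruster at the car's exhaust sockets.
# Each emitter derives its port count from filtered skeletal sockets; ribbons and sprites stream backward with inherited source movement, while the thruster is burst-spawned with looping lifetimes.
# Painted-variant overrides tint each layer, otherwise defaulting to blue and pink.
# A frontend-preview flag retunes velocities, colors and ribbon lifetime.
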NiagaraSystem /VehicleCosmetics/Boosts/Boost_Glitched/FX/NS_BoostGlitched
dependencies {
 usedDataInterfaces [
  &NiagaraDataInterfaceArrayFloat
  &NiagaraDataInterfaceCurve
  &NiagaraDataInterfaceArrayInt32
  &NiagaraDataInterfaceSkeletalMesh
  &NiagaraDataInterfaceArrayFloat3
 ]
 usedDynamicInputs [&SelectVectorFromArray]
 usedModules [
  &self:SystemSpawnScript.SystemScriptSource.SystemScriptGraph.NiagaraNodeAssignment_0.SetVariables_#1
  &self:SystemSpawnScript.SystemScriptSource.SystemScriptGraph.NiagaraNodeAssignment_1.SetVariables_#2
  &self:SystemSpawnScript.SystemScriptSource.SystemScriptGraph.NiagaraNodeAssignment_2.SetVariables_#3
  &SystemState
 ]
}
emitters {
 NE_VC_Boost_Ribbon_Parent {
  emitterUpdateScript [
   &EmitterState()
   &VC_FilteredSocketCount()
   &VC_RateQualityHandler()
   &VC_SpawnRibbon_v1(
    Number of Ribbons=$Emitter.PortCount
    SpawnRate=&MakeCustomFloatFromBool(Bool=$User.bIsFrontendPreview, True Float=60, False Float=60)
   )
  ]
  particleSpawnScript [
   &VC_PaintedVariantsOverride(
    Black=<rgba 0.005 0.005 0.005 2>
    BurntSienna=<rgba 0.01 0.0031 0.0012 1.5>
    Cobalt=<rgba 0 0.0089 0.4792 1>
    Crimson=<rgba 0.1354 0 0.0037 1>
    ForestGreen=<rgba 0 0.1719 0.0087 1>
    Grey=<rgba 0.0312 0.0312 0.0312 2>
    Lime=<rgba 0.1042 0.2083 0 1>
    Orange=<rgba 0.2 0.03 0 1>
    Purple=<rgba 0.0489 0 0.3913 1>
    Saffron=<rgba 0.2812 0.2156 0 1>
    TitaniumWhite=<rgba 0.4435 0.4435 0.4435 1>
   )
   &VC_InitializeRibbon_v1(
    Color=&MakeCustomLinearColorFromBool(
     Bool=$User.bIsPainted
     True Color=$Particles.Color
     False Color=&MakeCustomLinearColorFromBool(Bool=$User.bIsFrontendPreview, True Color=<rgba 0 0 1.5 1>, False Color=<rgba 0 0.05 2 1>)
    )
    Enable Skeletal Position=true
    Lifetime=&MakeCustomFloatFromBool(Bool=$User.bIsFrontendPreview, True Float=1.2, False Float=1)
    Ribbon Width=50
   )
   /SetVariables(Particles.SecondaryColor:LinearColor=&MakeCustomLinearColorFromBool(
     Bool=$User.bIsPainted
     True Color=&Multiply_LinearColorByFloat(
      Linear Color=$Particles.Color
      Float=&SelectFloatFromArray(Direct Array Index=$User.ColorVariantID, Float Selection Array=DataInterface() {propertyValues {FloatData [1,1,2,2,2,5,2,1,1,2.5,2]}})
     )
     False Color=&MakeCustomLinearColorFromBool(Bool=$User.bIsFrontendPreview, True Color=<rgba 1 0 0.4 1>, False Color=<rgba 1 0 0.5 1>)
    )) {isSetParameters=true}
   &AddVelocity(Velocity=&MakeCustomVectorFromBool(Bool=$User.bIsFrontendPreview, True Vector=<-2000 0 0>, False Vector=<-1200 0 0>)) {module=AddVelocity001}
   /SetVariables(Particles.SecondaryVisTag:NiagaraInt32=&MakeCustomIntFromBool(Bool=$User.bIsFrontendPreview, True INT=2, False INT=1)) {isSetParameters=true}
   /VehicleCosmetics/FX/VC_FXParents/Modules/DLR_ProductPreviews(GameModeID:NiagaraInt32=$System.GameModeID, bPreviewVehicleSpeed:NiagaraBool=$System.bPreviewVehicleSpeed)
  ]
  particleUpdateScript [
   &ParticleState()
   &InheritSourceMovement(
    Applied Position Scale=&VectorFromFloat(Value=&FloatFromCurve(FloatCurve=DataInterface() {propertyValues {Curve {
         defaultValue=1
         keys [
          {arriveTangent=-1.7885, interpMode=RCIM_Cubic, leaveTangent=-1.7885, tangentMode=RCTM_User, time=0.05, value=1}
          {interpMode=RCIM_Cubic, tangentMode=RCTM_User, time=1, value=0}
         ]
        }}}))
    ApplyForce=false
    ApplyPosition=true
    ApplyPositionFalloff=false
   )
   &SolveForcesAndVelocity(Clamp Velocity=false, Limit Acceleration=false)
   &VC_UpdateRibbon_v1(Enable Skeletal Position=true)
  ]
  rendererClasses [&NiagaraRibbonRendererProperties]
  renderers [
   {rendererClass=&NiagaraRibbonRendererProperties}
   {rendererClass=&NiagaraRibbonRendererProperties, rendererIndex=1}
   {rendererClass=&NiagaraRibbonRendererProperties, rendererIndex=2}
   {rendererClass=&NiagaraRibbonRendererProperties, rendererIndex=3}
   {rendererClass=&NiagaraRibbonRendererProperties, rendererIndex=4}
   {rendererClass=&NiagaraRibbonRendererProperties, rendererIndex=5}
   {rendererClass=&NiagaraRibbonRendererProperties, rendererIndex=6}
   {rendererClass=&NiagaraRibbonRendererProperties, rendererIndex=7}
   {rendererClass=&NiagaraRibbonRendererProperties, rendererIndex=8}
   {rendererClass=&NiagaraRibbonRendererProperties, rendererIndex=9}
   {rendererClass=&NiagaraRibbonRendererProperties, rendererIndex=10}
   {rendererClass=&NiagaraRibbonRendererProperties, rendererIndex=11}
   {rendererClass=&NiagaraRibbonRendererProperties, rendererIndex=12}
   {rendererClass=&NiagaraRibbonRendererProperties, rendererIndex=13}
   {rendererClass=&NiagaraRibbonRendererProperties, rendererIndex=14}
  ]
 }
 NE_VC_Boost_Sprite_Parent {
  emitterUpdateScript [
   &EmitterState()
   &VC_FilteredSocketCount()
   &SpawnRate(SpawnRate=50, Use Spawn Probability=false)
  ]
  particleSpawnScript [
   &VC_PaintedVariantsOverride(Cobalt=<rgba 0 0.074 4 1>, Purple=<rgba 0.163 0 1 1>)
   /SetVariables(Particles.VisibilityTag=&SelectIntFromArray(
     Array Sampling Mode=::NewEnumerator1
     Direct Array Index=?
     Int Selection Array=DataInterface() {propertyValues {IntData [0,1]}}
     Random Seed=0
    )) {isSetParameters=true}
   &InitializeParticle(
    Color=$Particles.Color
    Lifetime=&RandomRangeFloat(Minimum=0.4, Maximum=0.8)
    Sprite Rotation Mode=::NewEnumerator2
    Sprite Size Mode=::NewEnumerator3
    Uniform Sprite Size=&Multiply_Float(
     A=&RandomRangeFloat(Minimum=0.9, Maximum=1)
     B=&SelectFloatFromArray(Direct Array Index=$Particles.VisibilityTag, Float Selection Array=DataInterface() {propertyValues {FloatData [15,12]}})
    )
    UsePositionOffset=false
    Write Material Random=true
   )
   &PartitionParticles(Number Of Partitions=$Emitter.PortCount)
   &SkeletalMeshLocation(
    Bone / Socket Index=$Particles.PartitionParticles.Partition
    Bone / Socket Sampling Mode=::NewEnumerator5
    MaskSpawnGroup=false
    Mesh Sampling Type=::NewEnumerator6
    Skeletal Mesh=$System.SkeletalMesh
   )
   &AddVelocity(
    Cone Angle=10
    Module.UseConeFalloff=false
    Module.UseVelDistribution=false
    Override Seed=false
    Velocity Mode=::NewEnumerator2
    Velocity Speed=&MakeCustomFloatFromBool(Bool=$User.bIsFrontendPreview, True Float=-1400, False Float=-400)
   )
   &ShapeLocation(MaskSpawnGroup=false, Override Seed=false, Sphere Radius=44.737, Sphere Surface Distribution=1, UseOwnerScale=true)
  ]
  particleUpdateScript [
   &ParticleState()
   &InheritSourceMovement(Applied Position Scale=<0.5 0.5 0.5>, ApplyForce=false, ApplyPosition=true, ApplyPositionFalloff=false)
   &SolveForcesAndVelocity(Clamp Velocity=false, Limit Acceleration=false)
   &DynamicMaterialParameters(
    Index 1 Param 3=&MakeCustomFloatFromBool(Bool=$User.bIsPainted, True Float=1, False Float=0)
    Module.Index0Float0Write=false
    Module.Index0Float1Write=false
    Module.Index0Float2Write=false
    Module.Index0Float3Write=false
    Module.Index1Float0Write=false
    Module.Index1Float1Write=false
    Module.Index1Float2Write=true
    Module.Index1Float3Write=false
    Write Parameter Index 1=::NewEnumerator1
   )
  ]
  rendererClasses [&NiagaraSpriteRendererProperties]
  renderers [
   {rendererClass=&NiagaraSpriteRendererProperties}
   {rendererClass=&NiagaraSpriteRendererProperties, rendererIndex=1}
   {rendererClass=&NiagaraSpriteRendererProperties, rendererIndex=2}
   {rendererClass=&NiagaraSpriteRendererProperties, rendererIndex=3}
  ]
 }
 NE_VC_Thruster {
  emitterUpdateScript [
   &EmitterState(
    Inactive Response=::NewEnumerator1
    Life Cycle Mode=::NewEnumerator1
    Loop Behavior=::NewEnumerator1
    Loop Duration Mode=::NewEnumerator1
    UseLoopDelay=false
   )
   &VC_FilteredSocketCount()
   &SpawnRate(SpawnRate=30, Use Spawn Probability=false) {enabled=false}
   &SpawnBurst_Instantaneous(Spawn Count=2, Use Spawn Probability=false, UseLoopCountLimit=false)
   &SpawnBurst_Instantaneous(Spawn Count=2, Spawn Group=1, Use Spawn Probability=false, UseLoopCountLimit=false) {module=SpawnBurst_Instantaneous001}
  ]
  particleSpawnScript [
   &VC_PaintedVariantsOverride(Black=<rgba 0.0052 0.0052 0.0052 1>, Orange=<rgba 1.2 0.05 0 1>, Purple=<rgba 0.0801 0 0.6795 1>)
   &InitializeParticle(
    Color=&MakeCustomLinearColorFromBool(
     Bool=$User.bIsPainted
     True Color=$Particles.Color
     False Color=&MaskLinearColorBySpawnGroup(Masked Linear Color=<rgba 1 0.0781 0.6735 1>, Passthrough Linear Color=<rgba 0.0651 0.1014 1 1>, Spawn Group Mask=1)
    )
    Mesh Scale=&MaskVectorBySpawnGroup(Masked Vector=<2 2 2>, Passthrough Vector=<3 2 5>, Spawn Group Mask=1)
    Mesh Scale Mode=::NewEnumerator4
    Position Mode=::NewEnumerator1
    UsePositionOffset=false
    Write Material Random=true
   )
   &PartitionParticles(Number Of Partitions=$Emitter.PortCount)
   &SkeletalMeshLocation(
    Bone / Socket Index=$Particles.PartitionParticles.Partition
    Bone / Socket Sampling Mode=::NewEnumerator5
    Interpolate Mesh Sampling=false
    MaskSpawnGroup=false
    Mesh Sampling Type=::NewEnumerator6
    Sampled Position Offset=&AddVectorToPosition(Position=<0 0 0>, Vector=&AddVectorToPosition(Position=$User.DTCarOffset, Vector=<39.3897 0 0>))
    Skeletal Mesh=$System.SkeletalMesh
   )
   &AddVelocity(Velocity=&RandomRangeVector(Minimum=<-60 0 0>, Maximum=<-90 1 1>)) {enabled=false}
   &OffsetPosition(Position Offset=<-16.3255 0 -4>)
  ]
  particleUpdateScript [
   &ParticleState(Kill Particles When Lifetime Has Elapsed=false, Loop Particles Lifetime=true)
   &SolveForcesAndVelocity(Clamp Velocity=false, Limit Acceleration=false)
   &ScaleColor(
    Scale Alpha=&FloatFromCurve(FloatCurve=DataInterface() {propertyValues {Curve {keys [{arriveTangent=0.3506, time=0, value=1}, <at 6 0>]}}})
    ScaleA=true
    ScaleRGB=false
   )
   &DynamicMaterialParameters(
    Index 0 Param 2=1
    Index 1 Param 1=2
    Index 1 Param 2=0.8
    Index 1 Param 3=0.992
    Index 1 Param 4=&FloatFromCurve(FloatCurve=DataInterface() {propertyValues {Curve {keys [<at 0 2>,<at 1 0>]}}})
    Index 2 Param 1=0.7407
    Index 2 Param 2=0.6588
    Index 2 Param 3=&FloatFromCurve(FloatCurve=DataInterface() {propertyValues {Curve {keys [<at 0 0>,<at 1 1>]}}})
    Index 2 Param 4=0.016
    Index 3 Param 1=1.1907
    Module.Index0Param0Write=false
    Module.Index0Param1Write=true
    Module.Index0Param2Write=false
    Module.Index0Param3Write=false
    Module.Index1Param0Write=true
    Module.Index1Param1Write=true
    Module.Index1Param2Write=false
    Module.Index1Param3Write=true
    Module.Index2Param0Write=true
    Module.Index2Param1Write=true
    Module.Index2Param2Write=true
    Module.Index2Param3Write=true
    Module.Index3Param0Write=true
    Module.Index3Param1Write=true
    Module.Index3Param2Write=true
    Module.Index3Param3Write=true
    Write Parameter Index 1=true
    Write Parameter Index 2=true
    Write Parameter Index 3=true
   )
   &JitterPosition(Jitter Amount=0.3) {enabled=false}
  ]
  rendererClasses [&NiagaraMeshRendererProperties, &NiagaraSpriteRendererProperties]
  renderers [
   {rendererClass=&NiagaraMeshRendererProperties, rendererIndex=0}
   {rendererClass=&NiagaraMeshRendererProperties, rendererIndex=1}
   {rendererClass=&NiagaraSpriteRendererProperties, rendererIndex=2}
  ]
 }
}
summary {
 systemName=NS_BoostGlitched
 emitters [
  {emitterName=NE_VC_Boost_Ribbon_Parent, rendererClasses [&NiagaraRibbonRendererProperties]}
  {emitterName=NE_VC_Thruster, rendererClasses [&NiagaraMeshRendererProperties, &NiagaraSpriteRendererProperties]}
  {emitterName=NE_VC_Boost_Sprite_Parent, rendererClasses [&NiagaraSpriteRendererProperties]}
 ]
 userVariables [
  User.bIsFrontendPreview {type=NiagaraBool}
  User.bIsPainted {type=NiagaraBool}
  User.blsIsCoroColor {type=NiagaraBool}
  User.bPreviewVehicleSpeed {type=NiagaraBool}
  User.ColorVariantID {type=NiagaraInt32}
  User.DTCarOffset {type=NiagaraPosition, description="", defaultValue {}}
  User.GameModeID {type=NiagaraInt32, defaultValue=2}
  User.PaintedColor {type=LinearColor, description="", defaultValue {struct=LinearColor, value=<rgba 1 1 1 1>}}
  User.PaintedScalar {type=NiagaraFloat, defaultValue=1, description=""}
  User.PortCount {type=NiagaraInt32, defaultValue=2}
 ]
}
